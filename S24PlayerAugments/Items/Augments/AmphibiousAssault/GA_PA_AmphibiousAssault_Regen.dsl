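; Gameplay ability for the Amphibious Assault player augment that activates its effect while the player swims; the name points to regeneration.
; On activation it waits passively for the water tag to be added, activates the augment, then waits for the tag's removal.
; If the removal comes from a water jump, a short timer rechecks for the tag before ending the effect.
; Ending the effect re-arms the wait for swimming.
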(event Ability|EventActivateAbility
  ;[0.1171 1 0] Pasive, wait for swimming to begin
  (CallFunction|WaitforSwimming))

(event Custom|WaitForSwimming
  ;[1 0.2718 0.06097] Activate when swimming
  (Ability|Tasks|WaitGameplayTagAdd (Variables|Default|GetWaterTag) 0 true
    (:Added
      (CallFunction|ActivateAugment)
      (CallFunction|WaitforStopSwimming))))

(event Custom|WaitForStopSwimming
  ;[1 0 0.933] Wait for the player to begin swimming
  (Ability|Tasks|WaitGameplayTagRemove (Variables|Default|GetWaterTag) 0 true
    (:Removed
      (bind _outfortplayerpawn (CallFunction|GetAvatarasValidFortPlayerPawn)
        (:ValidFortPlayerPawn
          (if (Variables|/Script/FortniteGame.FortPlayerPawn|GetbIsWaterJump _outfortplayerpawn)
            (Utilities|Time|SetTimerbyEvent (EventDispatchers|CreateEvent :!event Custom|CheckForInWater) 1.0)
            (else
              (CallFunction|EndAugmentEffect))))))))

(event Custom|CheckForInWater
  ;[0 0.2854 1] Check if the player is back in the water after jumping out
  (bind _outfortplayerpawn (CallFunction|GetAvatarasValidFortPlayerPawn)
    (:ValidFortPlayerPawn
      (if (GameplayTags|HasMatchingGameplayTag _outfortplayerpawn (Variables|Default|GetWaterTag))
        (CallFunction|WaitforStopSwimming)
        (else
          (CallFunction|EndAugmentEffect))))))

(event EventEndAugmentEffect
  ;[1 0.5539 0] Setup the wait again when we leave water
  (|CallParentFunction)
  (CallFunction|WaitforSwimming))
